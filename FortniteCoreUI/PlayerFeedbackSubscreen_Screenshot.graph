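# Screenshot step of the player feedback modal, where the player reviews a captured screenshot before it goes with a report.
# It stacks a confirmation prompt, a Re-take Screenshot and Continue button row with gamepad input actions, and a fixed-size preview image.
# The widget is layout only: the prompt, buttons and preview bind to the native FortPlayerFeedbackModalSubscreen_Screenshot parent, and the event graph holds only unwired default events.
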WidgetBlueprint /FortniteCoreUI/PlayerFeedbackSubscreen_Screenshot
graphs [EventGraph {nodes [
   Event PreConstruct()
   Event Construct()
   Event Tick()
  ]}]
variables []
widgets {
 info {
  inheritedWidgetCount=4
  parentClass=/Script/FortniteUI.FortPlayerFeedbackModalSubscreen_Screenshot
  rootWidgetClass=&Overlay
  widgetCount=8
 }
 widgets [
  Overlay_3 {widgetClassPath=&Overlay, widget=Overlay_3}
  VerticalBox_3 {
   widgetClassPath=&VerticalBox
   parent=Overlay_3
   widget=VerticalBox_3
   slot {class=OverlaySlot, properties {horizontalAlignment=HAlign_Fill, verticalAlignment=VAlign_Fill}}
  }
  Text_LookingConfirmation {
   widgetClassPath=&CommonTextBlock
   bInherited=true
   parent=VerticalBox_3
   widget=Text_LookingConfirmation
   properties {
    autoWrapText=true
    mobileFontSizeMultiplier=1.3
    text=Make sure your screenshot captures what you want to report.
    font {
     bForceMonospaced=false
     bMaterialIsStencil=false
     fontMaterial=null
     fontObject=&HeadingNow
     letterSpacing=0
     monospacedWidth=1
     size=18
     skewAmount=0
     typefaceFontName=74 Regular
     outlineSettings {
      bApplyOutlineToDropShadows=false
      bMiteredCorners=false
      bSeparateFillAlpha=false
      outlineMaterial=null
      outlineSize=0
      outlineColor=<rgba 0 0 0 1>
     }
    }
    shadowOffset=<0 0>
   }
   slot {class=VerticalBoxSlot, properties {horizontalAlignment=HAlign_Center, padding=<ltrb 0 0 0 10>}}
  }
  HorizontalBox_2 {
   widgetClassPath=&HorizontalBox
   parent=VerticalBox_3
   widget=HorizontalBox_2
   slot {class=VerticalBoxSlot, properties {horizontalAlignment=HAlign_Center, padding=<ltrb 0 0 0 10>}}
  }
  Button_TakeScreenshot {
   widgetClassPath=&WBP_UIKit_Button_Regular
   bInherited=true
   bIsVariable=true
   parent=HorizontalBox_2
   widget=Button_TakeScreenshot
   properties {
    mobileMinHeight=88
    text=Re-take Screenshot
    triggeringInputAction {dataTable=&InputActionDataTable, rowName=GenericFaceButtonTop}
   }
   slot {class=HorizontalBoxSlot, properties {horizontalAlignment=HAlign_Left, verticalAlignment=VAlign_Top}}
  }
  Button_Continue {
   widgetClassPath=&WBP_UIKit_Button_Regular
   bInherited=true
   bIsVariable=true
   parent=HorizontalBox_2
   widget=Button_Continue
   properties {mobileMinHeight=88, text=Continue, triggeringInputAction {dataTable=&InputActionDataTable, rowName=GenericForward}}
   slot {class=HorizontalBoxSlot, properties {horizontalAlignment=HAlign_Left, verticalAlignment=VAlign_Top, padding=<ltrb 32 0 0 0>}}
  }
  SizeBox_6 {
   widgetClassPath=&SizeBox
   parent=VerticalBox_3
   widget=SizeBox_6
   properties {bOverride_HeightOverride=true, bOverride_WidthOverride=true, heightOverride=296, widthOverride=490}
   slot {class=VerticalBoxSlot, properties {horizontalAlignment=HAlign_Center, verticalAlignment=VAlign_Center}}
  }
  Image_Preview {widgetClassPath=&Image, bInherited=true, bIsVariable=true, parent=SizeBox_6, widget=Image_Preview, slot {class=SizeBoxSlot}}
 ]
}
